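# Physics asset for the male medium MasterKey NPC body, covering the pelvis, upper thighs and three dynamic pouch bones (dyn_puch), apparently for secondary motion on belt pouches.
# Each bone has one capsule, with both right pouches using identical capsule settings.
# All five other bodies constrain directly to the pelvis: pouches have tight swing and near-zero twist limits, while thigh twist bones allow a much wider range.
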PhysicsAsset /Game/Characters/Player/Male/Medium/Bodies/M_MED_MasterKey_NPC/Meshes/M_MED_Masterkey_NPC_Physics
bodies {
 dyn_puch_l [{shapeType=Capsule, length=11.6359, radius=2.8084, shapeName=dyn_puch_l_capsule, center=<2.3141 0 2.2859>, rotation=<rot 0 90.2343 -86.6261>}]
 dyn_puch_r_01 [{
   shapeType=Capsule
   length=10.3534
   radius=3.2037
   shapeName=dyn_puch_r_02_capsule
   center=<-2.1584 -0.164 -1.7167>
   rotation=<rot 0 -89.6496 -88.2865>
  }]
 dyn_puch_r_02 [{
   shapeType=Capsule
   length=10.3534
   radius=3.2037
   shapeName=dyn_puch_r_02_capsule
   center=<-2.1584 -0.164 -1.7167>
   rotation=<rot 0 -89.6496 -88.2865>
  }]
 pelvis [{shapeType=Capsule, length=14.75, radius=14.2736, center=<1.1102 0.1354 -0.4259>, rotation=<rot 0 -89.8768 -89.8034>}]
 thigh_twist_01_l [{shapeType=Capsule, length=19.3619, radius=7.7999, center=<9.2432 0 -1.1015>, rotation=<rot 81.0406 179.9999 180>}]
 thigh_twist_01_r [{shapeType=Capsule, length=19.3619, radius=7.7999, center=<-9.2432 0 1.1015>, rotation=<rot 81.0406 179.9999 180>}]
}
constraints [
 {bone1Name=thigh_twist_01_l, bone2Name=pelvis, swing1LimitDegrees=45, swing2LimitDegrees=45, twistLimitDegrees=45}
 {bone1Name=thigh_twist_01_r, bone2Name=pelvis, swing1LimitDegrees=45, swing2LimitDegrees=45, twistLimitDegrees=45}
 {bone1Name=dyn_puch_r_02, bone2Name=pelvis, swing1LimitDegrees=15, swing2LimitDegrees=10, twistLimitDegrees=1}
 {bone1Name=dyn_puch_l, bone2Name=pelvis, swing1LimitDegrees=15, swing2LimitDegrees=10, twistLimitDegrees=1}
 {bone1Name=dyn_puch_r_01, bone2Name=pelvis, swing1LimitDegrees=15, swing2LimitDegrees=10, twistLimitDegrees=1}
]
