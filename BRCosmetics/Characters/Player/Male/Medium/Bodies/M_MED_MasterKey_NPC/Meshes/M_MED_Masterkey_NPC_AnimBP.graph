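# Animation Blueprint for the MasterKey NPC medium male body, whose output pose is only a Copy Pose From Mesh of the attached parent.
# Unconnected to that output sit a Control Rig, zero-alpha pouch constraints and pelvis-space RigidBody chain, a second root-space RigidBody, and additive armor-flap bone offsets.
# The update event still maps each thigh-to-pelvis delta yaw into the left and right flap alphas those offsets read.
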AnimBlueprint /BRCosmetics/Characters/Player/Male/Medium/Bodies/M_MED_MasterKey_NPC/Meshes/M_MED_Masterkey_NPC_AnimBP
target_skeleton=/Game/Characters/Player/Male/Medium/Bodies/M_MED_MasterKey_NPC/Meshes/M_MED_Masterkey_NPC_Skeleton
graphs [
 AnimGraph {nodes [
   n1 = Control Rig() {
    alphaBoolBlend {blendOption=Linear}
    controlRigAssetReference {blueprintRigClass=/Game/Characters/Player/Male/Medium/Bodies/M_MED_MasterKey_NPC/Meshes/M_MED_Masterkey_NPC_CtrlRig.M_MED_Masterkey_NPC_CtrlRig_C}
   }
   n2 = "Constraint - dyn_mstr_puch_l (R)"(ComponentPose=Local To Component(LocalPose=@n1), Alpha=0, ConstraintWeights_0=1) {
    alphaBoolBlend {blendOption=Linear}
    boneToModify {boneName=dyn_mstr_puch_l}
    constraintSetup [{transformType=Rotation, perAxis {bY=false, bZ=false}, targetBone {boneName=thigh_twist_01_l}}]
    constraintWeights [1]
   }
   n3 = "Constraint - dyn_mstr_puch_r (RR)"!(ComponentPose=@n2, Alpha=0, ConstraintWeights_0=1, ConstraintWeights_1:float=0) -> (Pose) {
    alphaBoolBlend {blendOption=Linear}
    boneToModify {boneName=dyn_mstr_puch_r}
    constraintSetup [{transformType=Rotation, targetBone {boneName=thigh_r}}, {transformType=Rotation, targetBone {boneName=pelvis}}]
    constraintWeights [1,1]
   }
   n4 = RigidBody(ComponentPose=@n3) {
    bOverrideWorldGravity=true
    overridePhysicsAsset=/Game/Characters/Player/Male/Medium/Bodies/M_MED_MasterKey_NPC/Meshes/M_MED_Masterkey_NPC_Physics
    simulationSpace=BaseBoneSpace
    alphaBoolBlend {blendOption=Linear}
    baseBoneRef {boneName=pelvis}
    componentLinearAccScale=<0.1 0.1 0.1>
    componentLinearVelScale=<0.1 0.1 0.1>
    overrideWorldGravity {z=-1000}
   }
   Component To Local(ComponentPose=@n4)
   n5 = "Transform (Modify) Bone - Bone: dyn_armorflap_l"(Alpha=Get Flap l Alpha() -> (Flap l Alpha:double), Translation="-1,0,-0.000", Rotation="55,5,-80", Scale="1,1,1") {
    rotationMode=BMM_Additive
    translationMode=BMM_Additive
    alphaBoolBlend {blendOption=Linear}
    boneToModify {boneName=dyn_armorflap_l}
    rotation=<rot 55 5 -80>
    translation {x=-1}
   }
   "Transform (Modify) Bone - Bone: dyn_armorfrap_r"(ComponentPose=@n5, Alpha=Get Flap r Alpha() -> (Flap r Alpha:double), Translation="0,0,0", Rotation="-40,-10,-60", Scale="1,1,1") {
    rotationMode=BMM_Additive
    translationMode=BMM_Additive
    alphaBoolBlend {blendOption=Linear}
    boneToModify {boneName=dyn_armorfrap_r}
    rotation=<rot -40 -10 -60>
   }
   RigidBody() {
    bClampLinearTranslationLimitToRefPose=true
    bOverrideWorldGravity=true
    overridePhysicsAsset=&M_MED_Masterkey_NPC_Physics
    simulationSpace=BaseBoneSpace
    alphaBoolBlend {blendOption=Linear}
    baseBoneRef {boneName=root}
    componentAppliedLinearAccClamp=<3000 3000 3000>
    componentLinearAccScale=<0.35 0.35 0.45>
    componentLinearVelScale=<0.2 0.2 0.3>
    overrideWorldGravity {z=-980}
   }
   Output Pose(Result=Copy Pose From Mesh() {bUseAttachedParent=true}) {layerGroup=None}
   Get bIsCrouching()
  ]}
 EventGraph {nodes [
   Event BlueprintUpdateAnimation()
   n1 = BreakTransform(InTransform=GetDeltaTransformFromRefPose(self=GetOwnerSkeletalMeshComponent(), BoneName=pelvis, BaseName=thigh_l))
   > Set Flap l Alpha(Flap l Alpha:double=MapRangeClamped(Value=BreakRotator(InRot=@n1.Rotation).Yaw, InRangeA=20, InRangeB=90)) -> (Output_Get:double)
   n2 = BreakTransform(InTransform=GetDeltaTransformFromRefPose(self=GetOwnerSkeletalMeshComponent(), BoneName=pelvis, BaseName=thigh_r))
   > Set Flap r Alpha(Flap r Alpha:double=MapRangeClamped(Value=BreakRotator(InRot=@n2.Rotation).Yaw, InRangeA=20, InRangeB=90)) -> (Output_Get:double)
   TryGetPawnOwner()
  ]}
]
variables [
 Flap l Alpha {type=number}
 Flap r Alpha {type=number}
]
